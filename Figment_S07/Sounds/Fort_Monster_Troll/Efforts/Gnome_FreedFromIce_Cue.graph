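# Sound cue for the gnome's vocal effort when freed from ice, using the Troll sound class and attenuation settings.
# A quality-level switch sends its first level through a modulator to an equal-weight random pick of four takes.
# The other three levels share a modulator on take 01 alone, cutting variation.
# It overrides concurrency to block new instances at the limit and uses natural-sound distance falloff.
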SoundCue /Figment_S07/Sounds/Fort_Monster_Troll/Efforts/Gnome_FreedFromIce_Cue
root=QualityLevel_0
nodes [
 {id=Modulator_1, children [Random_1]}
 {id=Modulator_2, children [WavePlayer_10]}
 {id=QualityLevel_0, children [Modulator_1,Modulator_2,Modulator_2,Modulator_2]}
 {id=Random_1, children [WavePlayer_2,WavePlayer_3,WavePlayer_4,WavePlayer_9], properties {weights [1,1,1,1]}}
 {id=WavePlayer_10, properties {soundWaveAssetPtr=&/Gnome_FreedFromIce_01}}
 {id=WavePlayer_2, properties {soundWaveAssetPtr=&/Gnome_FreedFromIce_04}}
 {id=WavePlayer_3, properties {soundWaveAssetPtr=&Gnome_FreedFromIce_01}}
 {id=WavePlayer_4, properties {soundWaveAssetPtr=&/Gnome_FreedFromIce_02}}
 {id=WavePlayer_9, properties {soundWaveAssetPtr=&/Gnome_FreedFromIce_03}}
]
properties {
 attenuationSettings=/Figment_S07/Sounds/Attenuation/Fort_Monster_Troll_Att
 bOverrideConcurrency=true
 maxDistance=4600
 soundClassObject=/Figment_S07/Sounds/FortSoundClassesAndModes/Fort_Troll
 volumeMultiplier=1.2
 attenuationOverrides {distanceAlgorithm=NaturalSound, falloffDistance=5000, attenuationShapeExtents {x=1000}}
 concurrencyOverrides {resolutionRule=PreventNew}
}
